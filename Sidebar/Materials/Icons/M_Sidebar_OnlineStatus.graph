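# UI material that draws a procedural status dot icon, green by default, apparently for online status in the Sidebar.
# An MF_UI_Circle fill has a second, offset circle subtracted when Use2Circle is raised, allowing a crescent-like cutout.
# The saturated mask is multiplied by Icon Color alpha.
# UseCirleOutline blends two identical Fill outputs, so it has no effect as wired.
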Material /Sidebar/Materials/Icons/M_Sidebar_OnlineStatus
graph {
 expressions {
  MaterialExpressionAdd_2 {
   class=MaterialExpressionAdd
   inputs [
    A {from=MaterialExpressionMaterialFunctionCall_4, output=Normalized UV}
    B {from=MaterialExpressionConvert_0}
   ]
  }
  MaterialExpressionConvert_0 {
   class=MaterialExpressionConvert
   inputs [
    None {from=MaterialExpressionScalarParameter_3}
    None {from=MaterialExpressionScalarParameter_4}
   ]
   properties {
    nodeName=Make Vector2
    convertInputs [
     {
      type=Scalar
      expressionInput {
       expression=&self:MaterialExpressionScalarParameter_3
       inputName=None
       mask=0
       maskA=0
       maskB=0
       maskG=0
       maskR=0
       outputIndex=0
      }
     }
     {
      type=Scalar
      expressionInput {
       expression=&self:MaterialExpressionScalarParameter_4
       inputName=None
       mask=0
       maskA=0
       maskB=0
       maskG=0
       maskR=0
       outputIndex=0
      }
     }
    ]
    convertOutputs [{type=Vector2}]
   }
  }
  MaterialExpressionLinearInterpolate_0 {
   class=MaterialExpressionLinearInterpolate
   inputs [
    A {from=MaterialExpressionMaterialFunctionCall_0, output=Fill}
    B {from=MaterialExpressionMaterialFunctionCall_0, output=Fill}
    Alpha {from=MaterialExpressionScalarParameter_5}
   ]
  }
  MaterialExpressionLinearInterpolate_1 {
   class=MaterialExpressionLinearInterpolate
   inputs [
    A {from=""}
    B {from=MaterialExpressionMaterialFunctionCall_3, output=Fill}
    Alpha {from=MaterialExpressionScalarParameter_7}
   ]
  }
  MaterialExpressionMaterialFunctionCall_0 {
   class=MaterialExpressionMaterialFunctionCall
   inputs [
    UVs {from=MaterialExpressionMaterialFunctionCall_1, output=Normalized UV}
    Size {from=MaterialExpressionScalarParameter_0}
    OutlineThickness {from=MaterialExpressionScalarParameter_1}
    OutlineAlignment {from=""}
    GlowMin {from=""}
    GlowMax {from=""}
   ]
   properties {materialFunction=&MF_UI_Circle}
  }
  MaterialExpressionMaterialFunctionCall_1 {class=MaterialExpressionMaterialFunctionCall, inputs [], properties {materialFunction=&GetUserInterfaceUV}}
  MaterialExpressionMaterialFunctionCall_3 {
   class=MaterialExpressionMaterialFunctionCall
   inputs [
    UVs {from=MaterialExpressionAdd_2}
    Size {from=MaterialExpressionScalarParameter_2}
    OutlineThickness {from=""}
    OutlineAlignment {from=""}
    GlowMin {from=""}
    GlowMax {from=""}
   ]
   properties {materialFunction=&MF_UI_Circle}
  }
  MaterialExpressionMaterialFunctionCall_4 {class=MaterialExpressionMaterialFunctionCall, inputs [], properties {materialFunction=&GetUserInterfaceUV}}
  MaterialExpressionMultiply_1 {
   class=MaterialExpressionMultiply
   inputs [
    A {from=MaterialExpressionVectorParameter_1, output=A}
    B {from=MaterialExpressionSaturate_0}
   ]
  }
  MaterialExpressionSaturate_0 {class=MaterialExpressionSaturate, inputs [None {from=MaterialExpressionSubtract_0}]}
  MaterialExpressionScalarParameter_0 {class=MaterialExpressionScalarParameter, properties {defaultValue=0.74, group=IconShape, parameterName=CircleSize}}
  MaterialExpressionScalarParameter_1 {class=MaterialExpressionScalarParameter, properties {defaultValue=0.24, group=IconShape, parameterName=OutlineThickness}}
  MaterialExpressionScalarParameter_2 {class=MaterialExpressionScalarParameter, properties {defaultValue=0.7, group=IconShape, parameterName=2CircleSize, sortPriority=41}}
  MaterialExpressionScalarParameter_3 {class=MaterialExpressionScalarParameter, properties {defaultValue=0.2, group=IconShape, parameterName=2Circle_HorOffset, sortPriority=44}}
  MaterialExpressionScalarParameter_4 {class=MaterialExpressionScalarParameter, properties {defaultValue=0.2, group=IconShape, parameterName=2Circle_VerOffset, sortPriority=45}}
  MaterialExpressionScalarParameter_5 {class=MaterialExpressionScalarParameter, properties {group=IconShape, parameterName=UseCirleOutline}}
  MaterialExpressionScalarParameter_7 {class=MaterialExpressionScalarParameter, properties {group=IconShape, parameterName=Use2Circle, sortPriority=40}}
  MaterialExpressionSubtract_0 {
   class=MaterialExpressionSubtract
   inputs [
    A {from=MaterialExpressionLinearInterpolate_0}
    B {from=MaterialExpressionLinearInterpolate_1}
   ]
   properties {constB=0}
  }
  MaterialExpressionVectorParameter_1 {class=MaterialExpressionVectorParameter, properties {group=IconColor, parameterName=Icon Color, defaultValue=<rgba 0.0595 0.5711 0.1195 1>}}
 }
 properties {blendMode=BLEND_Translucent, materialDomain=MD_UI, pixelDepthOffsetMode=PDOM_Legacy, usedShadingModels=Default Lit}
}
