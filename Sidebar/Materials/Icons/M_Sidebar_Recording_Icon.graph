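# Translucent UI material for the Sidebar recording icon, drawing a rounded shape with button-state styling.
# A rounded-mask function builds the shape, inset by a pixel amount converted to normalized size, and multiplies it by a rerouted value.
# Modular-button functions chain base states into selectable states, declaring State Color and State Opacity reroutes.
# The base lerp's Color output feeds both base color and base opacity inputs of the selectable lerp.
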Material /Sidebar/Materials/Icons/M_Sidebar_Recording_Icon
graph {
 properties {bAutomaticallySetUsageInEditor=false, blendMode=BLEND_Translucent, materialDomain=MD_UI, usedShadingModels=Default Lit}
 nodes [
  n1 = Constant() {r=1}
  n2 = MaterialFunctionCall!(Value in Pixels=Constant() {r=8}, Pixel Reference) {materialFunction=/FortUILibrary/MaterialFunctions/Basics/MF_UI_PixelsToNormalizedV4}
  Multiply(
   A=MaterialFunctionCall(Size=Subtract(A=@n1, B=@n2."Normalized (V4)"), Corner Radius=Constant() {r=6}, Radius in Pixels=Constant() {r=1}) {materialFunction=&MF_UI_CoreUI_RoundedMask}.Shape Mask
   B=NamedRerouteUsage()
  )
  n3 = MaterialFunctionCall!(Idle, Idle Opacity, Hovered, Hovered Opacity, Focused, Focused Opacity, Pressed, Pressed Opacity, Disabled, Disabled Opacity) {materialFunction=/FortUILibrary/Systems/ModularButton/MaterialFunctions/MF_UIKit_ButtonStatesBase_Lerp}
  n4 = MaterialFunctionCall!(
   Base States=@n3.Color
   Base States Opacity=@n3.Color
   Selected_Idle
   Selected_Idle Opacity
   Selected_Hovered
   Selected_Hovered Opacity
   Selected_Focused
   Selected_Focused Opacity
   Selected_Pressed
   Selected_Pressed Opacity
   Selected_Disabled
   Selected_Disabled Opacity
  ) {materialFunction=/FortUILibrary/Systems/ModularButton/MaterialFunctions/MF_UIKit_ButtonStatesSelectable_Lerp}
  NamedRerouteDeclaration(None=@n4.Color) {name=State Color, nodeColor=<rgba 0 0 0 1>}
  NamedRerouteDeclaration(None=@n4.Opacity) {name=State Opacity, nodeColor=<rgba 0 0 0 1>}
  NamedRerouteUsage()
 ]
}
