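# M_Sidebar_VoiceActive is a translucent UI material for the Sidebar's voice-active icon, drawing animated audio-meter style stripes procedurally.
# Tiling Amount splits the normalized UI UV into columns, and MF_UI_Random with Seed gives per-column random values.
# Speed-scaled time drives sine oscillators that animate values and pattern offsets, while MF_UI_Box draws rounded segments with randomized height.
# Named reroutes link stages, including a mirrored stepped gradient, stripe opacity and a GlobalOpacity cap.
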Material /Sidebar/Materials/Icons/M_Sidebar_VoiceActive
graph {
 properties {bAutomaticallySetUsageInEditor=false, blendMode=BLEND_Translucent, materialDomain=MD_UI, usedShadingModels=Default Lit}
 nodes [
  Constant3Vector() {constant=<rgba 0.0395 1 0.2307 1>}
  n1 = NamedRerouteUsage()
  n2 = Multiply(A=Time(), B=ScalarParameter() {defaultValue=1, parameterName=Speed}) {desc=Speed}
  n3 = NamedRerouteUsage()
  n4 = ConstantBiasScale(None=Sine(None=Multiply(A=@n2, B=@n3)))
  ConstantBiasScale(None=Sine(None=LinearInterpolate(A=@n1, Alpha=@n4)))
  LinearInterpolate(Alpha=NamedRerouteUsage()) {constA=0.2, constB=0.1}
  LinearInterpolate(Alpha=NamedRerouteUsage()) {constA=-0.3, constB=-0.2}
  n5 = Round(None=ComponentMask(None=MaterialFunctionCall() {materialFunction=&GetUserInterfaceUV}.Normalized UV) {g=true})
  n6 = Multiply(A=NamedRerouteUsage()) {constB=0.1}
  n7 = NamedRerouteUsage()
  n8 = NamedRerouteUsage()
  n9 = Multiply(A=@n8) {constB=0.1125}
  n10 = Multiply(A=@n8) {constB=0.5}
  n11 = NamedRerouteUsage()
  n12 = AppendVector(
   A=NamedRerouteDeclaration(None=Constant() {r=0.8}) {name=Segment Width, nodeColor=<rgba 0 0 0 1>}
   B=NamedRerouteDeclaration(None=LinearInterpolate(A=@n9, B=@n10, Alpha=@n11)) {name=Random Scale, nodeColor=<rgba 0 0 0 1>}
  )
  n13 = Multiply(A=NamedRerouteUsage(), B=Constant() {desc=Corner Radius, r=1})
  n14 = Constant() {r=0.1}
  n15 = Constant() {r=-1}
  n16 = StaticBool() {value=true}
  n17 = Multiply(
   A=NamedRerouteDeclaration(None=MaterialFunctionCall(UVs=@n7, Size=@n12, CornerRoundness=@n13, OutlineThickness=@n14, OutlineAlignment=@n15, KeepAspectRatio=@n16) {materialFunction=&MF_UI_Box}.Fill) {name=Stripes, nodeColor=<rgba 0 0 0 1>}
   B=NamedRerouteUsage()
  )
  Max(A=LinearInterpolate(A=Multiply(A=@n5) {constB=0.2}, Alpha=@n6) {constB=2}, B=LinearInterpolate(A=@n6, Alpha=@n17))
  Max(A=@n1, B=@n4)
  n18 = NamedRerouteUsage()
  Min(
   A=Multiply(A=@n17, B=LinearInterpolate(Alpha=@n18) {constA=0.1, constB=0.2})
   B=ScalarParameter() {defaultValue=1, parameterName=GlobalOpacity}
  )
  Multiply(A=MaterialFunctionCall("Value (0-1)"=@n4) {materialFunction=&MF_UI_EaseCurves}.EaseInOutSine)
  n19 = Multiply(A=Time()) {desc=Speed}
  n20 = NamedRerouteUsage()
  n21 = ConstantBiasScale(None=Sine(None=Multiply(A=@n19, B=@n20)))
  n22 = Add(A=NamedRerouteUsage(), B=AppendVector(A=Constant(), B=LinearInterpolate(Alpha=@n21) {constA=-0.3, constB=0.1, desc=Offset Range})) {desc=Offset}
  n23 = Constant2Vector() {g=-0.125}
  n24 = NamedRerouteUsage()
  n25 = Divide(A=Subtract(A=Subtract(A=@n22, B=@n23)) {constB=0.5}, B=AppendVector(A=Constant() {r=1}, B=Divide(B=@n24)))
  n26 = DDX(Value=NamedRerouteDeclaration(None=Add(A=@n25) {constB=0.5}) {name=Animated Pattern UV, nodeColor=<rgba 0 0 0 1>})
  NamedRerouteDeclaration(None=Ceil(None=ComponentMask(None=@n26) {r=true})) {name=DDX Mask, nodeColor=<rgba 0 0 0 1>}
  NamedRerouteDeclaration(None=@n4) {name=Random Animated Values, nodeColor=<rgba 0 0 0 1>}
  n27 = ScalarParameter() {defaultValue=5, parameterName=Tiling Amount}
  n28 = MaterialFunctionCall() {materialFunction=&GetUserInterfaceUV}
  n29 = ComponentMask(None=@n28.Normalized UV) {r=true}
  n30 = AppendVector(
   A=Multiply(
    A=NamedRerouteDeclaration(None=@n27) {name=TilingX, nodeColor=<rgba 0 0 0 1>}
    B=NamedRerouteDeclaration(None=@n29) {name=Horizontal Gradient, nodeColor=<rgba 0 0 0 1>}
   )
   B=ComponentMask(None=@n28.Normalized UV) {g=true}
  )
  n31 = Floor(None=@n30)
  n32 = ScalarParameter() {defaultValue=28, parameterName=Seed}
  NamedRerouteDeclaration(None=MaterialFunctionCall!(UV=@n31, Seed=Multiply(A=@n32) {constB=0}) {materialFunction=&MF_UI_Random}.Result) {name=Random Columns_1, nodeColor=<rgba 0 0 0 1>}
  n33 = LinearInterpolate(Alpha=@n18) {constA=0.05, constB=0.5}
  NamedRerouteDeclaration(None=Multiply(A=@n17, B=Multiply(A=@n17, B=@n33))) {name=Stripe Opacity, nodeColor=<rgba 1 0 0.7969 1>}
  NamedRerouteDeclaration(None=MaterialFunctionCall(UV=@n31, Seed=@n32) {materialFunction=&MF_UI_Random}.Result) {name=Random Columns, nodeColor=<rgba 0 0 0 1>}
  n34 = NamedRerouteUsage()
  n35 = Min(A=@n34, B=OneMinus(None=@n34))
  n36 = NamedRerouteUsage()
  NamedRerouteDeclaration(None=MaterialFunctionCall!("Value (0-1)"=Multiply(A=@n35) {constB=2}, Steps=Multiply(A=@n36) {constB=0.5}) {materialFunction=/FortUILibrary/MaterialFunctions/Basics/MF_UI_StepRatio}.Stepped Gradient) {name=Segmented Gradient, nodeColor=<rgba 0 0 0 1>}
  NamedRerouteDeclaration(None=Frac(None=@n30)) {name=Pattern UV, nodeColor=<rgba 0 0 0 1>}
  NamedRerouteUsage()
  NamedRerouteUsage()
  NamedRerouteUsage()
 ]
}
